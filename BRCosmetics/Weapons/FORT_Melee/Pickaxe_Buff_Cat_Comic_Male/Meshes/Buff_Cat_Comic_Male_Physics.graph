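# Physics asset for the Buff Cat Comic Male two-handed pickaxe cosmetic, giving its top and bottom bar bones physics bodies.
# A capsule on the handle bone anchors two three-bone chains, barTop and barBottom, each bone carrying an identical short, wide capsule.
# Each chain links bone to bone back to the handle through constraints with very tight swing and twist limit angles.
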PhysicsAsset /BRCosmetics/Weapons/FORT_Melee/Pickaxe_Buff_Cat_Comic_Male/Meshes/Buff_Cat_Comic_Male_Physics
bodies {
 Melee_TwoHanded_Handle [{shapeType=Capsule, length=41.6534, radius=2.3792, center=<0.7881 -1.0324 18.3923>, rotation=<rot -0.105 -90 0>}]
 barBottom_01 [{shapeType=Capsule, length=6, radius=7.25, center=<0.1 0 0>, rotation=<rot 90 0 0>}]
 barBottom_02 [{shapeType=Capsule, length=6, radius=7.25, center=<0.1 0 0>, rotation=<rot 90 0 0>}]
 barBottom_03 [{shapeType=Capsule, length=6, radius=7.25, center=<0.1 0 0>, rotation=<rot 90 0 0>}]
 barTop_01 [{shapeType=Capsule, length=6, radius=7.25, center=<0.1 0 0>, rotation=<rot 90 0 0>}]
 barTop_02 [{shapeType=Capsule, length=6, radius=7.25, center=<0.1 0 0>, rotation=<rot 90 0 0>}]
 barTop_03 [{shapeType=Capsule, length=6, radius=7.25, center=<0.1 0 0>, rotation=<rot 90 0 0>}]
}
constraints [
 {bone1Name=barTop_01, bone2Name=Melee_TwoHanded_Handle, swing1LimitDegrees=2, swing2LimitDegrees=2, twistLimitDegrees=1}
 {bone1Name=barTop_02, bone2Name=barTop_01, swing1LimitDegrees=2, swing2LimitDegrees=2, twistLimitDegrees=1}
 {bone1Name=barTop_03, bone2Name=barTop_02, swing1LimitDegrees=2, swing2LimitDegrees=2, twistLimitDegrees=1}
 {bone1Name=barBottom_01, bone2Name=Melee_TwoHanded_Handle, swing1LimitDegrees=2, swing2LimitDegrees=2, twistLimitDegrees=1}
 {bone1Name=barBottom_02, bone2Name=barBottom_01, swing1LimitDegrees=2, swing2LimitDegrees=2, twistLimitDegrees=1}
 {bone1Name=barBottom_03, bone2Name=barBottom_02, swing1LimitDegrees=2, swing2LimitDegrees=2, twistLimitDegrees=1}
]
